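# Music Sound Cue for a Season 20 quest intro in Fortnite, named S20_Quest_NB_Intro.
# A single Wave Player is the root, playing the extended Origin intro track, named as 110 BPM, with no other nodes.
# It routes to the Fort_Music_Quest_PSM sound class with a slight volume boost and a very large max distance.
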SoundCue /Game/Sounds/Fort_Music/Seasonal/S20/S20_Quest_NB_Intro_Cue
root=WavePlayer_2
nodes [{id=WavePlayer_2, properties {soundWaveAssetPtr=&/FN_S20_Origin_Intro_Extended_110_BPM}}]
properties {maxDistance=2097152, soundClassObject=/Game/Sounds/FortSoundClassesAndModes/Fort_Music_Quest_PSM, volumeMultiplier=1.1}
